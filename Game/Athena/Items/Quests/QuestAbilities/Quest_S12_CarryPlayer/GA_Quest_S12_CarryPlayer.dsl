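; Quest ability that reports how far the owning player travels while carrying a player who is not on their team.
; On activation it finds the carried pawn among attached actors, checks teams and starts a repeating progress timer.
; Each tick sends the meters moved since the last sample, rounded, to the quest manager as a CarryPlayerDistanceTraveled stat event.
; Removal of the carrying tag clears the timer, sends a final update and ends the ability.
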(event Ability|EventActivateAbilityFromEvent
  (params (EventData :type Gameplay%20Event%20Data%20Structure%20%28by%20ref%29))
  (bind _carriedpawn (Variables|Default|GetCarriedPawn))
  (bind _owningplayerpawn (Variables|GAVars|Data|GetOwningPlayerPawn))
  (|CallParentFunction EventData)
  (for _array_element (Actor|GetAttachedActors _owningplayerpawn true false)
    (Utilities|IsValid _array_element
      (:Is%20Valid
        (bind _asfort_player_pawn_athena (Utilities|Casting|CastToFortPlayerPawnAthena _array_element))
        (Variables|Default|SetCarriedPawn _asfort_player_pawn_athena))))
  (Utilities|IsValid _carriedpawn
    (:Is%20Valid
      (if (not (Team|OnSameTeam _owningplayerpawn _carriedpawn))
        (Variables|GAVars|Data|SetPrevious%20Location (Transformation|GetActorLocation (Variables|GAVars|Data|GetOwningPlayerPawn)))
        ;[1 1 1] Start tracking loop
        (bind _returnvalue (Utilities|Time|SetTimerbyEvent (delegate Custom|ProgressQuest) (Variables|GAVars|Setup|GetProgressInterval) true))
        (Variables|GAVars|Data|SetTimer_Progress _returnvalue)
        ;[1 1 1] Clean up and end when we stop carrying
        (Ability|Tasks|WaitGameplayTagRemove (Variables|GAVars|Setup|GetT_CarryingPlayer)
          (:Removed
            (Utilities|Time|ClearandInvalidateTimerbyHandle (Variables|GAVars|Data|GetTimer_Progress))
            (CallFunction|ProgressQuest)
            (Ability|EndAbility)))))
    (:Is%20Not%20Valid
      _))
  (Ability|EndAbility))

(event Custom|ProgressQuest
  ;[1 1 1] Calc distance traveled in meters (100 Uu = 1m)
  (bind _output_get (Variables|GAVars|Data|SetCurrentLocation (Transformation|GetActorLocation (Variables|GAVars|Data|GetOwningPlayerPawn))))
  (bind _distance (Class|MissionBlueprintFunctionLibrary|DistanceBetweenTwoVectors _output_get (Variables|GAVars|Data|GetPrevious%20Location)))
  (if (> (Math|Float|Round (/ _distance 1e+02)) 0)
    ;[1 1 1] Send quest update
    (Quest|SendComplexCustomStatEvent (Variables|GAVars|Data|GetOwningPlayerQuestManager) 0 "(GameplayTags=((TagName=\"Athena.Quests.CarryPlayerDistanceTraveled\")))" "(GameplayTags=)" (Math|Float|Round (/ _distance 1e+02)))
    (else
      _))
  (Variables|GAVars|Data|SetPrevious%20Location (Variables|GAVars|Data|GetCurrentLocation)))
